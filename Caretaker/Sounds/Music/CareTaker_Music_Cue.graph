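# Sound cue for Caretaker NPC music, crossfading between ambient and action loops according to Unaware, Alerted and Aggressive float parameters.
# A mixer sums three continuous modulators, each scaling a looping wave's volume from its named parameter, silent by default.
# Unaware drives the ambient loop at a higher mixer volume, while Alerted and Aggressive both feed the same action loop player.
# It routes to the NPC music sound class.
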SoundCue /Caretaker/Sounds/Music/CareTaker_Music_Cue
root=Mixer_1
nodes [
 {id=Mixer_1, children [ModulatorContinuous_0, ModulatorContinuous_1, ModulatorContinuous_2], properties {inputVolume [2,1.5,1.5]}}
 {id=ModulatorContinuous_0, children [WavePlayer_0], properties {volumeModulationParams {default=0, parameterName=Unaware}}}
 {id=ModulatorContinuous_1, children [WavePlayer_1], properties {volumeModulationParams {default=0, parameterName=Alerted}}}
 {id=ModulatorContinuous_2, children [WavePlayer_1], properties {volumeModulationParams {default=0, parameterName=Aggressive}}}
 {id=WavePlayer_0, properties {bLooping=true, soundWaveAssetPtr=&/MX_Caretaker_Amb_Loop_01}}
 {id=WavePlayer_1, properties {bLooping=true, soundWaveAssetPtr=&/MX_Caretaker_Action_Loop_01}}
]
properties {maxDistance=2097152, soundClassObject=&Fort_Music_NPC, volumeMultiplier=1}
