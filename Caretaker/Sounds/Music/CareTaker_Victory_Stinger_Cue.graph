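# CareTaker_Victory_Stinger_Cue is a SoundCue that plays the Caretaker victory music stinger.
# A Delay node holds playback for a fixed five seconds, then feeds a single Wave Player for MX_Caretaker_Victory_Sting_01.
# The cue routes to the Fort_Music_NPC sound class, with boosted volume and a very large max distance.
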SoundCue /Caretaker/Sounds/Music/CareTaker_Victory_Stinger_Cue
root=Delay_0
nodes [
 {id=Delay_0, children [WavePlayer_2], properties {delayMax=5, delayMin=5}}
 {id=WavePlayer_2, properties {soundWaveAssetPtr=&/MX_Caretaker_Victory_Sting_01}}
]
properties {maxDistance=2097152, soundClassObject=&Fort_Music_NPC, volumeMultiplier=1.3}
